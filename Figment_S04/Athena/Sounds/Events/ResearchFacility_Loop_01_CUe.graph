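# Looping research facility sound cue for Fortnite Athena, in the Figment_S04 event sounds, routed to the Athena SFX sound class.
# A quality-level root feeds one looping wave player to its first two levels and leaves the other two empty.
# Attenuation is overridden inline with logarithmic distance falloff, a stereo spread and interpolated focus settings, and overall volume is slightly boosted.
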SoundCue /Figment_S04/Athena/Sounds/Events/ResearchFacility_Loop_01_CUe
root=QualityLevel_0
nodes [
 {id=QualityLevel_0, children [WavePlayer_1,WavePlayer_1,null,null]}
 {id=WavePlayer_1, properties {bLooping=true, soundWaveAssetPtr=&/researchfacility_loop_01}}
]
properties {
 bOverrideAttenuation=true
 maxDistance=2048
 soundClassObject=/Game/Sounds/FortSoundClassesAndModes/Fort_Events_SoundClass/Fort_Athena_SFX
 volumeMultiplier=1.25
 attenuationOverrides {
  bEnableFocusInterpolation=true
  distanceAlgorithm=Logarithmic
  falloffDistance=1024
  focusAttackInterpSpeed=0.1
  focusAzimuth=3
  focusReleaseInterpSpeed=0.5
  nonFocusAzimuth=7
  nonFocusVolumeAttenuation=0
  occlusionInterpolationTime=2
  occlusionTraceChannel=ECC_GameTraceChannel11
  occlusionVolumeAttenuation=0
  stereoSpread=512
  attenuationShapeExtents {x=1024}
 }
}
